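# Sound cue for the impact sound of the Balloon pickaxe cosmetic.
# Picks one of nine equally weighted balloon impact waves, applies a slight random upward pitch variation, and plays it after a short fixed delay.
# Uses the shared pickaxe impact attenuation, property sheet and Impacts sound class, with melee weapon concurrency, and is primed on load.
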SoundCue /BRCosmetics/Sounds/PickAxes/Balloon/Pickaxe_Balloon_Impact_Cue
root=Delay_1
nodes [
 {id=Delay_1, children [Modulator_0], properties {delayMax=0.1, delayMin=0.1}}
 {id=Modulator_0, children [Random_2], properties {pitchMax=1.2, pitchMin=1}}
 {
  id=Random_2
  children [WavePlayer_12, WavePlayer_13, WavePlayer_14, WavePlayer_15, WavePlayer_16, WavePlayer_17, WavePlayer_18, WavePlayer_19, WavePlayer_20]
  properties {weights [1,1,1,1,1,1,1,1,1]}
 }
 {id=WavePlayer_12, properties {soundWaveAssetPtr=&/PA_Balloon_Impact_03}}
 {id=WavePlayer_13, properties {soundWaveAssetPtr=&/PA_Balloon_Impact_09}}
 {id=WavePlayer_14, properties {soundWaveAssetPtr=&/PA_Balloon_Impact_01}}
 {id=WavePlayer_15, properties {soundWaveAssetPtr=&/PA_Balloon_Impact_02}}
 {id=WavePlayer_16, properties {soundWaveAssetPtr=&/PA_Balloon_Impact_04}}
 {id=WavePlayer_17, properties {soundWaveAssetPtr=&/PA_Balloon_Impact_05}}
 {id=WavePlayer_18, properties {soundWaveAssetPtr=&/PA_Balloon_Impact_06}}
 {id=WavePlayer_19, properties {soundWaveAssetPtr=&/PA_Balloon_Impact_07}}
 {id=WavePlayer_20, properties {soundWaveAssetPtr=&/PA_Balloon_Impact_08}}
]
properties {
 attenuationSettings=/Game/Sounds/Attenuation/Fort_Pickaxe_Impact_Att
 audioPropertiesSheet=/BRCosmetics/Sounds/Settings/PropertySheets/Pickaxe/APS_Cosmetic_Pickaxe_Impact
 bPrimeOnLoad=true
 maxDistance=4300
 soundClassObject=&Fort_Impacts
 attenuationOverrides {
  bAttenuateWithLPF=true
  distanceAlgorithm=Logarithmic
  falloffDistance=800
  lPFRadiusMax=800
  lPFRadiusMin=400
  attenuationShapeExtents {x=300}
 }
 concurrencyOverrides {bLimitToOwner=true, maxCount=2, resolutionRule=StopOldest}
 concurrencySet [&Player_Melee_Weapon_Conc]
}
